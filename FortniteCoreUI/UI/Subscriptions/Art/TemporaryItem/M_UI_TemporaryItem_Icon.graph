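# UI material for the temporary item subscription icon, drawing an animated countdown timer from the icon texture.
# Two radial arc masks, vertically offset, run on looping Time staggered half a period apart with cubed easing.
# Their stepped difference cuts into the texture's green channel, which is added to the red-channel shape masked by inverted green.
# TimerSpeedMult scales the sweep, Overall_Alpha scales only the red-channel shape, and a white constant supplies color.
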Material /FortniteCoreUI/UI/Subscriptions/Art/TemporaryItem/M_UI_TemporaryItem_Icon
graph {
 properties {
  bUsedWithStaticMesh=false
  blendMode=BLEND_Translucent
  materialDomain=MD_UI
  pixelDepthOffsetMode=PDOM_Legacy
  usedShadingModels=Default Lit
 }
 nodes [
  n1 = TextureSample(UVs=TextureCoordinate()) {texture=&/T_UI_TempItemIcon}
  n2 = OneMinus(None=@n1.G)
  n3 = TextureCoordinate()
  n4 = AppendVector(A=ComponentMask(None=@n3) {r=true}, B=Subtract(A=ComponentMask(None=@n3) {g=true}) {constB=0.085})
  n5 = Subtract(A=Time() {bOverride_Period=true, period=4}) {constB=2}
  n6 = ScalarParameter() {defaultValue=0.5, parameterName=TimerSpeedMult}
  n7 = Power(Base=OneMinus(None=Multiply(A=@n5, B=@n6))) {constExponent=3}
  n8 = TextureCoordinate()
  n9 = AppendVector(A=ComponentMask(None=@n8) {r=true}, B=Subtract(A=ComponentMask(None=@n8) {g=true}) {constB=0.085})
  n10 = Time() {bOverride_Period=true, period=4}
  n11 = ScalarParameter() {defaultValue=0.5, parameterName=TimerSpeedMult}
  n12 = Power(Base=OneMinus(None=Multiply(A=@n10, B=@n11))) {constExponent=3}
  n13 = Subtract(
   A=Step(Y=MaterialFunctionCall!(UV=@n4, "Arc Start Angle (0-1)", "Arc End Angle (0-1)", "Fill Ratio (0-1)"=@n7, Fill Clockwise) {materialFunction=/FortUILibrary/MaterialFunctions/Basics/MF_UI_RadialArc}.Radial Fill) {constX=0}
   B=Step(Y=MaterialFunctionCall(UV=@n9, "Fill Ratio (0-1)"=@n12) {materialFunction=&MF_UI_RadialArc}.Radial Fill) {constX=0}
  )
  Add(
   A=Multiply(A=LinearInterpolate(B=@n1.R, Alpha=@n2), B=ScalarParameter() {defaultValue=1, parameterName=Overall_Alpha})
   B=Saturate(None=Subtract(A=@n1.G, B=@n13))
  )
  Constant3Vector() {constant=<rgba 1 1 1 1>}
 ]
}
