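# WBP_STI_Label is a temporary-item label widget in the subscriptions UI, showing an icon and text on a material-backed border.
# PreConstruct passes the LabelType enum and current text to Setup Label, which switches between default, warning and sneak-peek styles.
# Each style calls Apply Effect to rebuild the border brush and set text, colors and icon.
# Default uses white content, the others near-black, and sneak peek collapses the icon.
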WidgetBlueprint /FortniteCoreUI/UI/Subscriptions/WBP_STI_Label
graphs [
 Apply Effect {nodes [
   n1 = Apply Effect!() -> (
    then
    Text:Text
    TextColor:Slate Color Structure
    BackgroundMaterial:Material Instance Object Reference
    IconVisibility:ESlateVisibility Enum=Visible
    IconMaterial:Material Object Reference
    IconColor:Linear Color Structure
   ) {is=title K2Node_FunctionEntry}
   > SetBrush(
    self:Border Object Reference=Get Border_Background() -> (Border_Background:Common Border Object Reference)
    InBrush=Make Slate Brush!(
     TintColor:Slate Color Structure=(SpecifiedColor=(R=1,G=1,B=1,A=1),ColorUseRule=UseColor_Specified)
     DrawAs:ESlateBrushDrawType Enum=Box
     Tiling:ESlateBrushTileType Enum=NoTile
     Mirroring:ESlateBrushMirrorType Enum=NoMirror
     ImageSize:Vector 2D Structure=MakeVector2D(X=64, Y=64)
     Margin:Margin Structure=Make Margin(Left=0.2, Top=0, Right=0.2, Bottom=0)
     ResourceObject:Object Reference=@n1.BackgroundMaterial
     OutlineSettings:Slate Brush Outline Settings Structure=(CornerRadii=(X=0,Y=0,Z=0,W=0),Color=(SpecifiedColor=(R=0,G=0,B=0,A=0),ColorUseRule=UseColor_Specified),Width=0,RoundingType=HalfHeightRadius,bUseBrushTransparency=False)
    ) -> (SlateBrush:Slate Brush Structure) {is=title K2Node_MakeStruct}
   )
   > Set Text(Text=@n1, self=Get LabelText() -> (LabelText:Common Text Object Reference))
   > SetColorAndOpacity(self=Get LabelText(), InColorAndOpacity=@n1.TextColor)
   > SetVisibility(self=Get LabelIcon() -> (LabelIcon:Image Object Reference), InVisibility=@n1.IconVisibility)
   > SetBrushFromMaterial(self=Get LabelIcon(), Material=@n1.IconMaterial)
   > Set ColorAndOpacity!(execute, ColorAndOpacity:Linear Color Structure=@n1.IconColor, self:Image Object Reference=Get LabelIcon()) -> (then, Output_Get:Linear Color Structure)
  ]}
 EventGraph {nodes [
   Event PreConstruct()
   > Setup Label!(
    execute
    self
    Selection:ETempItemLabelType Enum=Get LabelType() -> (LabelType:ETempItemLabelType Enum=NewEnumerator0)
    Text:Text=Get Text(self:Text Object Reference=Get LabelText())
   ) -> (then)
  ]}
 Setup Label {nodes [
   n1 = Setup Label!() -> (then, Selection:ETempItemLabelType Enum=NewEnumerator0, Text:Text) {is=title K2Node_FunctionEntry}
   > n2 = Switch on ETempItemLabelType!(execute, Selection:ETempItemLabelType Enum=@n1) -> (NewEnumerator0:Exec, NewEnumerator7:Exec, NewEnumerator1:Exec) {is=title K2Node_SwitchEnum}
   > Apply Effect!(
    execute
    self
    Text:Text=Get Text!() -> (Text)
    TextColor:Slate Color Structure=Make Slate Color(SpecifiedColor=(R=1,G=1,B=1,A=1))
    BackgroundMaterial:Material Instance Object Reference=/FortniteCoreUI/UI/Subscriptions/Art/TemporaryItem/M_UI_TempItemTimerBackground_Default
    IconVisibility:ESlateVisibility Enum=Visible
    IconMaterial:Material Object Reference=/FortniteCoreUI/UI/Subscriptions/Art/TemporaryItem/M_UI_TemporaryItem_Icon
    IconColor:Linear Color Structure=(R=1,G=1,B=1,A=1)
   ) -> (then) {is=title K2Node_CallFunction}
   Apply Effect(
    execute=@n2.NewEnumerator1
    Text=Get Text()
    TextColor=Make Slate Color(SpecifiedColor=(R=0.0021,G=0.0021,B=0.0021,A=1))
    BackgroundMaterial=/FortniteCoreUI/UI/Subscriptions/Art/TemporaryItem/M_UI_TempItemTimerBackground_Warning
    IconVisibility=Visible
    IconMaterial=&M_UI_TemporaryItem_Icon
    IconColor=(R=0.0021,G=0.0021,B=0.0021,A=1)
   )
   Apply Effect(
    execute=@n2.NewEnumerator7
    Text=Get Text()
    TextColor=Make Slate Color(SpecifiedColor=(R=0.0021,G=0.0021,B=0.0021,A=1))
    BackgroundMaterial=/FortniteCoreUI/UI/Subscriptions/Art/TemporaryItem/M_UI_TempItemTimerBackground_SneakPeek
    IconVisibility=Collapsed
    IconMaterial=&M_UI_TemporaryItem_Icon
    IconColor=(R=0.0021,G=0.0021,B=0.0021,A=1)
   )
  ]}
]
variables [LabelType {type=ETempItemLabelType, default=CTL_Default}]
widgets {
 info {inheritedWidgetCount=0, parentClass=&UserWidget, rootWidgetClass=&CommonBorder, widgetCount=4}
 widgets [
  Border_Background {
   widgetClassPath=&CommonBorder
   bIsVariable=true
   widget=Border_Background
   properties {
    background {
     bIsDynamicallyLoaded=false
     drawAs=Box
     imageType=NoImage
     mirroring=NoMirror
     resourceName=None
     resourceObject=&M_UI_TempItemTimerBackground_Warning
     tiling=NoTile
     imageSize=<64 64>
     margin=<ltrb 0.2 0 0.2 0>
     outlineSettings {
      bUseBrushTransparency=false
      roundingType=HalfHeightRadius
      width=0
      color {colorUseRule=UseColor_Specified, specifiedColor=<rgba 0 0 0 0>}
      cornerRadii=<0 0 0 0>
     }
     tintColor {colorUseRule=UseColor_Specified, specifiedColor=<rgba 1 1 1 1>}
     uVRegion {bIsValid=false, max=<0 0>, min=<0 0>}
    }
    padding=<ltrb 18 10 18 11>
   }
  }
  HBox_Timer {
   widgetClassPath=&HorizontalBox
   parent=Border_Background
   widget=HBox_Timer
   slot {class=BorderSlot, properties {padding=<ltrb 18 10 18 11>}}
  }
  LabelIcon {
   widgetClassPath=&Image
   bIsVariable=true
   parent=HBox_Timer
   widget=LabelIcon
   properties {brush {
     bIsDynamicallyLoaded=false
     drawAs=Image
     imageType=NoImage
     mirroring=NoMirror
     resourceName=None
     resourceObject=&M_UI_TemporaryItem_Icon
     tiling=NoTile
     imageSize=<24 24>
     margin=<ltrb 0 0 0 0>
     outlineSettings {
      bUseBrushTransparency=false
      roundingType=HalfHeightRadius
      width=0
      color {colorUseRule=UseColor_Specified, specifiedColor=<rgba 0 0 0 0>}
      cornerRadii=<0 0 0 0>
     }
     tintColor {colorUseRule=UseColor_Specified, specifiedColor=<rgba 1 1 1 1>}
     uVRegion {bIsValid=false, max=<0 0>, min=<0 0>}
    }}
   slot {class=HorizontalBoxSlot, properties {horizontalAlignment=HAlign_Right, verticalAlignment=VAlign_Center, padding=<ltrb -5 0 5 0>}}
  }
  LabelText {
   widgetClassPath=&CommonTextBlock
   bIsVariable=true
   parent=HBox_Timer
   widget=LabelText
   properties {
    mobileFontSizeMultiplier=1.4
    text=AVAILABLE UNTIL 00/00
    colorAndOpacity {colorUseRule=UseColor_Foreground, specifiedColor=<rgba 1 0 1 1>}
    font {
     bForceMonospaced=false
     bMaterialIsStencil=false
     fontMaterial=null
     fontObject=&BurbankSmall
     letterSpacing=0
     monospacedWidth=1
     size=14
     skewAmount=0
     typefaceFontName=Bold
     outlineSettings {
      bApplyOutlineToDropShadows=false
      bMiteredCorners=false
      bSeparateFillAlpha=false
      outlineMaterial=null
      outlineSize=0
      outlineColor=<rgba 0 0 0 1>
     }
    }
    shadowOffset=<0 0>
   }
   slot {class=HorizontalBoxSlot, properties {verticalAlignment=VAlign_Center}}
  }
 ]
}
